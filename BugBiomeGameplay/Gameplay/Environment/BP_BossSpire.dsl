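; Boss spire environment actor for the bug biome, whose Active state switches its persistent effects and swarm audio on or off.
; OnRep_Active toggles the FX component's visibility with its children, fades the fly, close-fly and ground swarm loops in or out, and prints a debug message.
; The construction script calls the parent, then matches the child actor's root visibility to Active, which the RepNotify does not update directly.
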(fn OnRep_Active
  (bind _flyswarm_loop (Variables|Audio|GetFlySwarm_Loop))
  (bind _groundswarm_loop (Variables|Audio|GetGroundSwarm_Loop))
  (bind _flyswarm_close_loop (Variables|Audio|GetFlySwarm_Close_Loop))
  (bind _persistentfxcomponent (Variables|Default|GetPersistentFXComponent))
  (if (Variables|Default|GetActive)
    (Rendering|SetVisibility _persistentfxcomponent true true)
    (Development|PrintString "Activate Spire FX" true true "(R=0.000000,G=0.660000,B=1.000000,A=1.000000)" 1e+01)
    (Audio|Components|Audio|FadeIn _flyswarm_loop 0.5 1.0 0.0 "Linear" :!class /Script/Engine.AudioComponent)
    (Audio|Components|Audio|FadeIn _groundswarm_loop 0.5 1.0 0.0 "Linear" :!class /Script/Engine.AudioComponent)
    (Audio|Components|Audio|FadeIn _flyswarm_close_loop 0.5 1.0 0.0 "Linear" :!class /Script/Engine.AudioComponent)
    (else
      (Rendering|SetVisibility _persistentfxcomponent false true)
      (Development|PrintString "Deactivate Spire FX" true true "(R=0.000000,G=0.660000,B=1.000000,A=1.000000)" 1e+01)
      (Audio|Components|Audio|FadeOut _flyswarm_loop 0.5 0.0 "Linear" :!class /Script/Engine.AudioComponent)
      (Audio|Components|Audio|FadeOut _groundswarm_loop 0.5 0.0 "Linear" :!class /Script/Engine.AudioComponent)
      (Audio|Components|Audio|FadeOut _flyswarm_close_loop 0.5 0.0 "Linear" :!class /Script/Engine.AudioComponent))))

(fn ConstructionScript
  (|CallParentFunction)
  (Utilities|FlowControl|Sequence
    (:then_0
      (if (Variables|Default|GetActive)
        (Rendering|SetVisibility (Variables|/Script/Engine.Actor|GetRootComponent (Variables|/Script/Engine.ChildActorComponent|GetChildActor (Variables|Default|GetChildActor))) true true)
        (else
          (Rendering|SetVisibility (Variables|/Script/Engine.Actor|GetRootComponent (Variables|/Script/Engine.ChildActorComponent|GetChildActor (Variables|Default|GetChildActor))) false true))))))
